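# Unlit translucent cloud material for meshes and mesh particles, lit by time-of-day colors and a volcano glow.
# A panning, flattened normal texture drives sun shading, scaled by a custom Schlick phase function for scattering.
# A sphere-masked directional glow around VolcanoLocation follows MPC_Snow lava and smoke parameters.
# The normal also builds a time-of-day vertex offset, while opacity combines view-angle, UV-edge and near-camera fades, with lobby and particle-color switches.
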Material /Figment_S08/Athena/Environments/Materials/Cloud/M_Figment_Volcano_Cloud
graph {
 properties {
  bAutomaticallySetUsageInEditor=false
  bUsedWithMeshParticles=true
  bUsedWithNiagaraMeshParticles=true
  bUsedWithSplineMeshes=true
  blendMode=BLEND_Translucent
  shadingModel=MSM_Unlit
  usedShadingModels=Unlit
 }
 nodes [
  n1 = Multiply(A=CollectionParameter() {collection=/Game/Athena/Prototype/Blueprints/White/MPC_Snow, parameterName=SmokeGlowAdd}) {constB=3}
  n2 = Constant() {r=0.2}
  n3 = Add(
   A=Multiply(A=CollectionParameter() {collection=&MPC_Snow, parameterName=LavaFlowsVis}) {constB=0.5}
   B=MaterialFunctionCall(Value_Morning=Constant() {r=0.1}, Value_Day=Constant() {r=0.05}, Value_Evening=@n2, Value_Night=@n2) {materialFunction=&Scalar_Time_of_Day}.Result
  ) {constB=0.2}
  n4 = WorldPosition()
  n5 = VectorParameter() {parameterName=VolcanoLocation, defaultValue=<rgba 84500 36944 20000 0>}
  n6 = SphereMask(A=@n4.XYZ, B=@n5.RGB, Radius=ScalarParameter() {defaultValue=45000, parameterName=VolcanoGlowRadius}) {hardnessPercent=0}
  n7 = Subtract(A=@n5.RGB, B=@n4.XYZ)
  n8 = VertexNormalWS()
  n9 = TextureCoordinate()
  n10 = AppendVector(
   A=Multiply(A=ScalarParameter() {defaultValue=2, parameterName=Tile_N_X}, B=ComponentMask(None=@n9) {r=true})
   B=Multiply(A=ComponentMask(None=@n9) {g=true}, B=ScalarParameter() {defaultValue=6, parameterName=Tile_N_Y})
  )
  n11 = ComponentMask(None=VectorParameter() {parameterName=Speed, defaultValue=<rgba 0.008 0.0125 0 0>}.RGB) {g=true, r=true}
  n12 = MaterialFunctionCall(
   Normal=TextureSample(UVs=Panner(Coordinate=@n10, Speed=@n11)) {samplerType=SAMPLERTYPE_LinearColor, texture=/Game/VisualThreatManager/StormVisuals/LargeClouds/T_StormCloudMobile}.RGB
   Flatness=ScalarParameter() {defaultValue=0.5, parameterName=FlattenDetailNormal}
  ) {materialFunction=&FlattenNormal}
  n13 = Transform(None=@n12.Result)
  n14 = Saturate(None=DotProduct(A=Normalize(VectorInput=@n7), B=LinearInterpolate(A=@n8, B=@n13)))
  n15 = Multiply(
   A=Multiply(A=Add(A=@n1, B=@n3), B=VectorParameter() {parameterName=VolcanoGlowTint, defaultValue=<rgba 3 0.6801 0.3 0>}.RGB)
   B=Power(Base=Multiply(A=@n6, B=@n14))
  )
  n16 = ParticleColor()
  n17 = VectorParameter() {parameterName=ColorLobby1, defaultValue=<rgba 1 0.7828 0.6615 1>}
  n18 = VectorParameter() {parameterName=ColorLobby2, defaultValue=<rgba 0.1455 0.2074 0.2812 1>}
  n19 = CollectionParameter() {
   collection=&FortniteMaterialParameters
   parameterName=SunlightDirectionalVector
  }
  n20 = ComponentMask(None=@n19) {b=true, g=true, r=true}
  n21 = VertexNormalWS()
  n22 = DotProduct(
   A=Transform(None=Normalize(VectorInput=@n20)) {transformSourceType=TRANSFORMSOURCE_World, transformType=TRANSFORM_Tangent}
   B=LinearInterpolate(A=@n12.Result, B=LinearInterpolate(A=@n12.Result, B=@n21), Alpha=ScalarParameter() {defaultValue=0.1, parameterName=LerpVertNormal})
  )
  n23 = Clamp(None=Power(Base=ConstantBiasScale(None=@n22) {bias=0.5, scale=0.6}, Exp=ScalarParameter() {defaultValue=2.4, parameterName=SoftenShadow}))
  n24 = MaterialFunctionCall(
   Color_Morning=VectorParameter() {parameterName=LightColorMorning, defaultValue=<rgba 0.8992 0.5725 0.9062 1>}.RGB
   Color_Day=VectorParameter() {parameterName=LightColorDay, defaultValue=<rgba 0.8177 0.945 1 1>}.RGB
   Color_Evening=VectorParameter() {parameterName=LightColorEvening, defaultValue=<rgba 0.7969 0.4746 0.303 1>}.RGB
   Color_Night=VectorParameter() {parameterName=LightColorNight, defaultValue=<rgba 0.2019 0.2816 0.316 1>}.RGB
  ) {materialFunction=&Color_Time_of_Day}
  n25 = Multiply(
   A=MaterialFunctionCall(
    Color_Morning=VectorParameter() {parameterName=ShadowColorDay, defaultValue=<rgba 0.1679 0.1885 0.2049 1>}.RGB
    Color_Day=VectorParameter() {parameterName=ShadowColorMorning, defaultValue=<rgba 0.1552 0.1846 0.191 1>}.RGB
    Color_Evening=VectorParameter() {parameterName=ShadowColorEvening, defaultValue=<rgba 0.1181 0.0709 0.0828 1>}.RGB
    Color_Night=VectorParameter() {parameterName=ShadowColorNight, defaultValue=<rgba 0.0952 0.1258 0.1319 1>}.RGB
   ) {materialFunction=&Color_Time_of_Day}.Result
   B=@n24.Result
  )
  n26 = CollectionParameter() {
   collection=&FortniteMaterialParameters
   parameterName=SunlightDirectionalVector
  }
  MaterialExpressionDotProduct_0 = DotProduct(A=ComponentMask(None=@n26) {b=true, g=true, r=true}, B=CameraVectorWS())
  MaterialExpressionScalarParameter_7 = ScalarParameter() {defaultValue=0.4, parameterName=Anisotropy}
  n27 = Custom!(CosTheta=@MaterialExpressionDotProduct_0, k=@MaterialExpressionScalarParameter_7) {
   code="\tfloat Inner = (1 + k * CosTheta);\r\n\treturn (1 - k * k) / (4 * PI * Inner * Inner);\r\n\r\n"
   description=Schlick Phase
   inputs [
    {
     inputName=CosTheta
     input {expression=&self:MaterialExpressionDotProduct_0}
    }
    {inputName=k, input {expression=&self:MaterialExpressionScalarParameter_7}}
   ]
  }
  n28 = Multiply(
   A=StaticSwitchParameter(True=LinearInterpolate(A=@n17.RGB, B=@n18.RGB, Alpha=@n23), False=LinearInterpolate(A=@n25, B=@n24.Result, Alpha=@n23)) {parameterName=IsLobby}
   B=Multiply(A=@n27, B=ScalarParameter() {defaultValue=28, parameterName=ScaleCloudColor})
  )
  Add(A=@n15, B=StaticSwitchParameter(True=Multiply(A=@n28, B=@n16.RGB), False=@n28) {parameterName=Use Particle Color})
  n29 = Constant()
  FeatureLevelSwitch(Default=@n15, ES2=@n29, ES3_1=@n29)
  n30 = Transform(None=Multiply(A=@n12.Result, B=VectorParameter() {parameterName=Offset, defaultValue=<rgba 150 150 150 0>}.RGB))
  n31 = ScalarParameter() {defaultValue=1.35, parameterName=OffsetScale_Morning}
  n32 = VectorParameter() {parameterName=CloudScaleByTOD, defaultValue=<rgba 0 0 1 0>}
  n33 = ScalarParameter() {defaultValue=0.05, parameterName=OffsetScaleDay}
  n34 = ScalarParameter() {defaultValue=2, parameterName=OffsetScaleEvening}
  n35 = ScalarParameter() {defaultValue=2.5, parameterName=OffsetScaleNight}
  n36 = ComponentMask(None=MaterialFunctionCall(
    Color_Morning=Multiply(A=@n31, B=@n32.RGB)
    Color_Day=Multiply(A=@n33, B=@n32.RGB)
    Color_Evening=Multiply(A=@n34, B=@n32.RGB)
    Color_Night=Multiply(A=@n35, B=@n32.RGB)
   ) {materialFunction=&Color_Time_of_Day}.Result) {b=true}
  n37 = MaterialFunctionCall!(Width=Constant() {r=0.475}, Sharpness, Offset=Constant(), Direction Switch=StaticBool() {value=true}, Compare, Input Coordinates) {materialFunction=/Engine/Functions/Engine_MaterialFunctions03/Procedurals/GeneratedBand}
  Multiply(A=Multiply(A=Multiply(A=@n30, B=@n36), B=VertexColor().R), B=Saturate(None=Multiply(A=@n37.Result) {constB=1.5}))
  n38 = Fresnel(ExponentIn=ScalarParameter() {defaultValue=0.7283, parameterName=FresnelLobby}, Normal=@n13) {baseReflectFraction=0}
  n39 = ScalarParameter() {defaultValue=0.3, parameterName=Opacity}
  n40 = ComponentMask(None=CameraVectorWS()) {b=true}
  n41 = ScalarParameter() {defaultValue=0.01, parameterName=offset}
  n42 = Clamp(None=Divide(A=Add(A=@n40, B=@n41), B=ScalarParameter() {defaultValue=0.2, parameterName=Div}))
  n43 = LinearInterpolate(A=Multiply(A=@n39, B=ScalarParameter() {defaultValue=2, parameterName=OpacityUnder}) {constB=4}, B=@n39, Alpha=@n42)
  n44 = MaterialFunctionCall("FadeSensitivity (optional)"=ScalarParameter() {defaultValue=3, parameterName=AngleFadeSensitivityU}) {materialFunction=&FadeBasedOnViewAngle}
  n45 = MaterialFunctionCall("FadeSensitivity (optional)"=ScalarParameter() {defaultValue=2.4, parameterName=AngleFadeSensitivity}) {materialFunction=&FadeBasedOnViewAngle}
  n46 = ComponentMask(None=TextureCoordinate()) {g=true}
  n47 = Multiply(A=@n46, B=OneMinus(None=@n46))
  n48 = ScalarParameter() {defaultValue=2, parameterName=ScaleCloudTipRamp}
  n49 = Multiply(
   A=LinearInterpolate(A=Multiply(A=@n43, B=@n44.Result), B=Multiply(A=@n43, B=@n45.Result), Alpha=@n42)
   B=Clamp(None=Multiply(A=@n47, B=@n48))
  )
  n50 = MaterialFunctionCall(Distance=ScalarParameter() {defaultValue=1200, parameterName=CameraDistanceFade}) {materialFunction=&Near_Camera_Fade}
  n51 = StaticSwitchParameter(
   True=Multiply(A=OneMinus(None=@n38), B=ScalarParameter() {defaultValue=0.708, parameterName=OpacityLobby})
   False=Multiply(A=Multiply(A=@n49, B=@n50.Result), B=VertexColor())
  ) {parameterName=IsLobby}
  StaticSwitchParameter(True=Multiply(A=@n16.A, B=@n51), False=@n51) {parameterName=Use Particle Color}
 ]
}
